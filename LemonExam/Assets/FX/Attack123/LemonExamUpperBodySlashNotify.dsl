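; Anim notify state for the LemonExam upper-body slash that repositions its spawned Niagara effect when the notify begins.
; After the parent spawns the effect on a FortPawn, it compares the root socket's Y axis with actor forward.
; If nearly aligned, it shifts the effect horizontally by the pelvis-to-root offset.
; Otherwise it swings the effect around the pelvis by the signed angle about root up and rotates it toward the projected actor forward.
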(fn Received_NotifyBegin
  (params (MeshComp :type Skeletal%20Mesh%20Component%20Object%20Reference) (Animation :type Anim%20Sequence%20Base%20Object%20Reference) (TotalDuration :type Float%20%28single-precision%29) (EventReference :type Anim%20Notify%20Event%20Reference%20Structure%20%28by%20ref%29))
  (bind _returnvalue (Transformation|GetSocketRotation MeshComp "root"))
  (bind (_x _y _z) (Math|Rotator|GetAxes _returnvalue))
  (bind _returnvalue_2 (Transformation|GetSocketLocation MeshComp "Pelvis"))
  (bind _returnvalue_3 (Components|GetOwner MeshComp))
  (bind _returnvalue_4 (Transformation|GetActorForwardVector _returnvalue_3))
  (bind _returnvalue_5 (Math|Vector|ProjectVectorontoPlane _returnvalue_4 _z))
  (bind (_x_2 _y_2) (Math|Rotator|GetAxes _returnvalue))
  (bind _returnvalue_6 (Math|Vector|Normalize _returnvalue_5))
  (bind _returnvalue_7 (|CallParentFunction MeshComp Animation TotalDuration EventReference))
  (bind _returnvalue_8 (AnimNotify|GetSpawnedEffect MeshComp))
  (bind _asniagara_particle_system_component (Utilities|Casting|CastToNiagaraComponent _returnvalue_8)
    (:then
      (Utilities|Casting|CastToFortPawn (Components|GetOwner MeshComp)
        (:then
          (if (<= (Class|AbilityFunctions|GetAngleBetweenTwoVectors _returnvalue_4 _y_2 _returnvalue_3) 5.0)
            (Transformation|AddWorldOffset _asniagara_particle_system_component (* (- (Transformation|GetSocketLocation MeshComp "root") _returnvalue_2) "1.000000,1.000000,0.000000") false false)
            (else
              (Transformation|SetWorldLocation _asniagara_particle_system_component (+ (Math|Vector|RotateVectorAroundAxis (- (Transformation|GetWorldLocation _asniagara_particle_system_component) _returnvalue_2) (* (Class|AbilityFunctions|GetAngleBetweenTwoVectors _y _returnvalue_6 _asniagara_particle_system_component) (Math|Float|SelectFloat 1.0 -1.0 (<= (Math|Vector|DotProduct _returnvalue_6 _x) 0))) _z) _returnvalue_2) false false)
              (Transformation|SetWorldRotation _asniagara_particle_system_component (Math|Rotator|CombineRotators (Transformation|GetWorldRotation _asniagara_particle_system_component) (Math|Conversions|ToRotator(Quat) (Math|Quat|FindQuatBetweenVectors _y _returnvalue_6))) false false)))
          (return _returnvalue_7))
        (:CastFailed
          _)))
    (:CastFailed
      _))
  (return _returnvalue_7))
