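# Niagara sword-slash effect for LemonExam's Attack123 FX set; its only emitter, Minimal, is disabled as saved.
# That emitter burst-spawns tall, bright lavender sprites with explicitly written facing and vertical Z alignment.
# A cubic curve drives alpha, rising quickly and then fading out over particle life.
# Dynamic material parameters and a random material value are also written for the sprite material.
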NiagaraSystem /LemonExam/Assets/FX/Attack123/NS_LemonExam_SlashSwordEffect
dependencies {usedDataInterfaces [&NiagaraDataInterfaceCurve], usedDynamicInputs [], usedModules [&SystemState]}
emitters {Minimal {
  bEnabled=false
  emitterUpdateScript [
   &EmitterState(Inactive Response=::NewEnumerator1, Life Cycle Mode=::NewEnumerator1, Loop Behavior=::NewEnumerator1, UseLoopDelay=false)
   &SpawnBurst_Instantaneous(Use Spawn Probability=false, UseLoopCountLimit=false)
  ]
  particleSpawnScript [&InitializeParticle(
    Color=<rgba 1.4437 1.2029 2 1>
    Sprite Size=<107.8801 304.204>
    Sprite Size Mode=::NewEnumerator4
    UsePositionOffset=false
    Write Material Random=true
   )]
  particleUpdateScript [
   &ParticleState()
   &SpriteFacingAndAlignment(Sprite Alignment=<0 0 1>, WriteAlignment=true, WriteFacing=true)
   &DynamicMaterialParameters(
    Index 0 Param 1=1
    Index 0 Param 2=1
    Index 0 Param 3=1
    Index 0 Param 4=1
    Module.Index0Float0Write=true
    Module.Index0Float1Write=true
    Module.Index0Float2Write=true
    Module.Index0Float3Write=true
   )
   &ScaleColor(
    Scale Alpha=&FloatFromCurve(FloatCurve=DataInterface() {propertyValues {Curve {keys [
         {interpMode=RCIM_Cubic, leaveTangent=15.0385, tangentMode=RCTM_Break, time=0, value=0}
         {interpMode=RCIM_Cubic, tangentMode=RCTM_Break, time=0.1618, value=1}
         {arriveTangent=-1.1931, interpMode=RCIM_Cubic, leaveTangent=-1.1931, time=0.6751, value=0.6673}
         <at 1 0>
        ]}}})
    ScaleA=true
    ScaleRGB=true
   )
  ]
  rendererClasses [&NiagaraSpriteRendererProperties]
  renderers [{rendererClass=&NiagaraSpriteRendererProperties}]
 }}
summary {
 systemName=NS_LemonExam_SlashSwordEffect
 emitters [{bEnabled=false, emitterName=Minimal, rendererClasses [&NiagaraSpriteRendererProperties]}]
 userVariables []
}
