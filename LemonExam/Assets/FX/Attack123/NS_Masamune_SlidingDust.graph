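# Niagara dust trail for Masamune's sliding, filed under Attack123, spawning per unit of movement at skeletal mesh bones or sockets.
# Particle partitions select the bone or socket and, for meshes, flip orientation between two sides.
# The Mesh emitter makes brief, scaling, fading dust meshes.
# The Smoke_SphereCenter emitter makes SubUV smoke sprites that inherit some velocity, settle under gravity and drag, grow, fade and tint by time of day.
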NiagaraSystem /LemonExam/Assets/FX/Attack123/NS_Masamune_SlidingDust
dependencies {
 usedDataInterfaces [
  &NiagaraDataInterfaceCurve
  &NiagaraDataInterfaceSpriteRendererInfo
  &NiagaraDataInterfaceColorCurve
  &NiagaraDataInterfaceArrayFloat
  &NiagaraDataInterfaceSkeletalMesh
 ]
 usedDynamicInputs [&Vector2DFromFloat]
 usedModules [
  &self:SystemSpawnScript.SystemScriptSource.SystemScriptGraph.NiagaraNodeAssignment_0.SetVariables_#1
  &SystemState
 ]
}
emitters {
 Mesh {
  emitterUpdateScript [
   &EmitterState()
   &SpawnPerUnit(Module.Use Max Movement Threshold=false, Module.Use Movement Tolerance=false, Spawn Spacing=9, Use Spawn Probability=false)
  ]
  particleSpawnScript [
   &PartitionParticles()
   &InitializeParticle(
    Color=<rgba 0.758 0.6029 0.4778 1>
    Lifetime Max=0.1
    Lifetime Min=0.08
    Lifetime Mode=::NewEnumerator1
    Mesh Scale Max=<2 1.2 1>
    Mesh Scale Min=<0.5 0.6 1>
    Mesh Scale Mode=::NewEnumerator2
    UsePositionOffset=false
    Write Material Random=true
   )
   &CameraOffset(Camera Offset Amount=10)
   &InitialMeshOrientation(
    Mesh Axis To Orient=&MakeVector(
     X=&SelectFloatFromArray(Direct Array Index=$Particles.PartitionParticles.Partition, Float Selection Array=DataInterface() {propertyValues {FloatData [1,-1]}})
     Y=0
     Z=0
    )
    Mesh Orientation Mode=::NewEnumerator2
    Rotation=<0 0.5 0>
    Use Rotation Vector=true
   )
   &SkeletalMeshLocation(
    Bone / Socket Index=$Particles.PartitionParticles.Partition
    Bone Sampling Mode=::NewEnumerator4
    MaskSpawnGroup=false
    Skeletal Mesh=$System.SK_Mesh
   )
  ]
  particleUpdateScript [
   &ParticleState()
   &ScaleColor(
    Scale Alpha=&FloatFromCurve(
     FloatCurve=DataInterface() {propertyValues {Curve {keys [
         {arriveTangent=-0.7355, interpMode=RCIM_Cubic, leaveTangent=-0.7355, tangentMode=RCTM_User, time=0.5, value=1}
         {arriveTangent=-3.1363, interpMode=RCIM_Cubic, leaveTangent=-3.1363, tangentMode=RCTM_User, time=1, value=0}
        ]}}}
     Scale Curve=0.4
    )
    ScaleA=true
    ScaleRGB=true
   )
   &ScaleMeshSize(Scale Factor=&Multiply_VectorByFloat(
     Vector=&MakeVector(
      X=&FloatFromCurve(FloatCurve=DataInterface() {propertyValues {Curve {keys [{arriveTangent=2.7289, interpMode=RCIM_Cubic, leaveTangent=2.7289, tangentMode=RCTM_User, time=0, value=0}, <at 1 1>]}}})
      Y=&FloatFromCurve(FloatCurve=DataInterface() {propertyValues {Curve {keys [{arriveTangent=2.7289, interpMode=RCIM_Cubic, leaveTangent=2.7289, tangentMode=RCTM_User, time=0, value=0}, <at 1 1>]}}})
      Z=1
     )
     Float=0.5
    ))
  ]
  rendererClasses [&NiagaraMeshRendererProperties]
  renderers [{rendererClass=&NiagaraMeshRendererProperties, rendererIndex=0}]
 }
 Smoke_SphereCenter {
  emitterUpdateScript [
   &EmitterState()
   &SpawnPerUnit(Module.Use Max Movement Threshold=false, Module.Use Movement Tolerance=false, Spawn Spacing=15, Use Spawn Probability=false)
  ]
  particleSpawnScript [
   &PartitionParticles()
   &InitializeParticle(
    Color=<rgba 0.758 0.6029 0.4778 1>
    Lifetime=&RandomRangeFloat(Minimum=0.3, Maximum=0.5)
    Sprite Rotation Angle=&RandomRangeFloat(Minimum=-25, Maximum=25)
    Sprite Rotation Mode=::NewEnumerator0
    Sprite Size Mode=::NewEnumerator3
    Uniform Sprite Size=&RandomRangeFloat(Minimum=20, Maximum=40)
    UsePositionOffset=false
    Write Material Random=true
   )
   &SkeletalMeshLocation(
    Bone / Socket Index=$Particles.PartitionParticles.Partition
    Bone Sampling Mode=::NewEnumerator4
    MaskSpawnGroup=false
    Skeletal Mesh=$System.SK_Mesh
   )
   &InheritVelocity(Inherited Velocity Amount Scale=<0.2 0.2 0.2>, Speed Limit=true, UseSpeedThreshold=false)
  ]
  particleUpdateScript [
   &ParticleState()
   &GravityForce(Gravity=&RandomRangeVector(Minimum=<0 0 -12.2821>, Maximum=<0 0 -71.5186>))
   &Drag(
    Drag=&MergeHelper_FloatSwitch(
     0=&RandomRangeFloat(Minimum=1.2, Maximum=1.3)
     1=&RandomRangeFloat(Minimum=1.5, Maximum=2)
     2=0
     3=0
     4=0
     Index=&SpawnGroup()
     Num=::NewEnumerator3
    )
    Use Linear Drag=true
    Use Rotational Drag=true
   )
   &SolveForcesAndVelocity(Clamp Velocity=false, Limit Acceleration=false)
   &ScaleSpriteSize(Uniform Curve Sprite Scale=DataInterface() {propertyValues {Curve {keys [
        {arriveTangent=3.9789, interpMode=RCIM_Cubic, leaveTangent=3.9789, tangentMode=RCTM_User, time=-0.0085, value=0.6876}
        {arriveTangent=1.3013, interpMode=RCIM_Cubic, leaveTangent=1.3013, time=0.299, value=1.7288}
        {arriveTangent=0.0224, interpMode=RCIM_Cubic, leaveTangent=0.0224, tangentMode=RCTM_User, time=1, value=2}
       ]}}})
   &DynamicMaterialParameters(
    Index 0 Param 1=&Lerp_Float(
     A=1
     B=2
     Alpha=&FloatFromCurve(FloatCurve=DataInterface() {propertyValues {Curve {keys [<at 0 0 RCIM_Cubic>,<at 1 1 RCIM_Cubic>]}}})
    )
    Module.Index0Param0Write=true
    Module.Index0Param1Write=true
    Module.Index0Param2Write=true
    Module.Index0Param3Write=true
   )
   &SubUVAnimation(
    Frames Per Second:NiagaraInt32=&MergeHelper_IntSwitch(
     0=&RandomRangeInt(Minimum=10, Maximum=15)
     1=20
     2=0
     3=0
     4=0
     5=?
     Index=&SpawnGroup()
     Num=::NewEnumerator3
    )
    Playback Mode:ENiagara_FPSPlayrate=::NewEnumerator1
    Playback Scale Curve=DataInterface() {propertyValues {Curve {keys [
        {interpMode=RCIM_Cubic, tangentMode=RCTM_User, time=0, value=3}
        {arriveTangent=-1.2656, interpMode=RCIM_Cubic, leaveTangent=-1.2656, tangentMode=RCTM_User, time=1, value=0}
       ]}}}
    Random Start Frame=true
    "Sprite Renderer "=DataInterface() {propertyValues {SpriteRenderer=&self:Smoke_SphereCenter_0.NiagaraSpriteRendererProperties_1}}
    Start Frame Offset:NiagaraInt32=&RandomRangeInt(Minimum=0, Maximum=6)
    Start Frame Range Override=&RandomRangeInt(Minimum=0, Maximum=0)
    SubUV Animation Mode:ENiagara_SubUVLookupModeV2=::NewEnumerator3
    UseEndFrame=true
    UsePlayRateScale=true
    UseStartFrame=true
   )
   &ScaleColor(
    Scale RGB=&MakeVectorFromLinearColorRGB(Color=&FNTimeOfDayColor(
      Night Color=<rgba 0.6 0.7243 1 1>
      Morning Color=<rgba 0.7 0.6725 0.6692 1>
      Day Color=<rgba 1 0.8103 0.65 1>
      Evening Color=<rgba 0.8 0.7149 0.6539 1>
     ))
    ScaleA=true
    ScaleRGB=true
   ) {module=ScaleColor001}
   &ScaleColor(
    Scale Alpha=&FloatFromCurve(
     FloatCurve=DataInterface() {propertyValues {Curve {keys [
         {arriveTangent=0.7573, interpMode=RCIM_Cubic, leaveTangent=0.7573, tangentMode=RCTM_User, time=0, value=0}
         {arriveTangent=-0.3388, interpMode=RCIM_Cubic, leaveTangent=-0.3388, tangentMode=RCTM_User, time=0.02, value=1}
         {arriveTangent=-2.7546, interpMode=RCIM_Cubic, leaveTangent=-2.7546, tangentMode=RCTM_User, time=1, value=0}
        ]}}}
     Scale Curve=0.2
    )
    Scale RGB=&MakeVectorFromLinearColorRGB(Color=&ColorFromCurve(ColorCurve=DataInterface() {propertyValues {
        AlphaCurve {keys [<at 0 1 RCIM_Cubic>]}
        BlueCurve {keys [<at 0.0872 0.5>]}
        GreenCurve {keys [<at 0.0872 0.5>]}
        RedCurve {keys [<at 0.0872 0.5>]}
       }}))
    ScaleA=true
    ScaleRGB=true
   ) {module=ScaleColor002}
   &CameraOffset()
  ]
  rendererClasses [&NiagaraSpriteRendererProperties]
  renderers [{rendererClass=&NiagaraSpriteRendererProperties}]
 }
}
summary {
 systemName=NS_Masamune_SlidingDust
 emitters [
  {emitterName=Smoke_SphereCenter, rendererClasses [&NiagaraSpriteRendererProperties]}
  {emitterName=Mesh, rendererClasses [&NiagaraMeshRendererProperties]}
 ]
 userVariables [User.IsSpawnDust {type=NiagaraBool, defaultValue=true}]
}
